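; Cheat manager for Spires gameplay, providing debug commands that act on the spires closest to the player's pawn.
; The major spire command loops over all Spire living world prefabs by squared distance, then starts prediction on the closest one's prediction component to trigger activation.
; The minor spire command finds the nearest attacking minor spire and changes its state, apparently to vulnerable.
; Both commands run only with authority.
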(fn GetClosestLivingWorldPrefab
  (bind _returnvalue (CheatManager|GetPlayerController))
  (bind _returnvalue_2 (Pawn|GetControlledPawn _returnvalue))
  (bind _outactors (Actor|GetAllActorsOfClass "/SpiresGameplay/Lager/LWPrefabs/LWP_Spire.LWP_Spire_C"))
  (for _array_element _outactors
    (if (< (Transformation|GetSquaredDistanceTo _array_element _returnvalue_2) (LocalVariables|GetClosestDistance))
      (LocalVariables|SetClosestDistance (Transformation|GetSquaredDistanceTo _array_element _returnvalue_2))
      (LocalVariables|SetClosestSpireLWP _array_element)))
  (return (LocalVariables|GetClosestSpireLWP)))

(fn MajorSpire_TriggerClosestSpireActivation
  (if (Networking|HasAuthority (CheatManager|GetPlayerController))
    (bind _closestspireprefab (CallFunction|GetClosestLivingWorldPrefab))
    (Utilities|IsValid _closestspireprefab
      (:Is%20Valid
        (Class|FortLWPPredictionComponent|StartPrediction (Variables|/SpiresGameplay/Lager/LWPrefabs/LWP_Spire.LWP_Spire_C|GetBP_LWPredictionComponent_Spires _closestspireprefab))))))

(fn MinorSpire_Vulnerable
  (bind _returnvalue (CheatManager|GetPlayerController))
  (if (Networking|HasAuthority _returnvalue)
    (bind _outactors (Actor|GetAllActorsOfClass "/SpireBoss/Gameplay/MinorSpire/BGA_Spire_Minor_Attacking.BGA_Spire_Minor_Attacking_C"))
    (bind _asbga_spire_minor_attacking (Utilities|Casting|CastToBGA_Spire_Minor_Attacking (.ReturnValue (Actor|FindNearestActor (Transformation|GetActorLocation (Pawn|GetControlledPawn _returnvalue)) _outactors))))
    (Variables|/SpireBoss/Gameplay/MinorSpire/BGA_Spire_Minor.BGA_Spire_Minor_C|SetSpireState "NewEnumerator2" _asbga_spire_minor_attacking)))
